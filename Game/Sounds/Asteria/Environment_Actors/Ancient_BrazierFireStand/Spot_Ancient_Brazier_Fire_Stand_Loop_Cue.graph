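# Looping ambient sound cue for the Ancient Brazier fire stand environment actor.
# A mixer layers three looping fire waves reused from Winterfest fireplace, Curie fire and Coliseum brazier sounds, each behind its own attenuation node.
# The first two quality slots wrap that mix in a second mixer that adds the Coliseum branch again; the last two play it directly.
# Concurrency is overridden to two instances with distance-based volume scaling.
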SoundCue /Game/Sounds/Asteria/Environment_Actors/Ancient_BrazierFireStand/Spot_Ancient_Brazier_Fire_Stand_Loop_Cue
root=QualityLevel_1
nodes [
 {
  id=Attenuation_1
  children [WavePlayer_1]
  properties {attenuationSettings=&/Spot_Bazier_Fire_Close_Att}
 }
 {
  id=Attenuation_2
  children [WavePlayer_6]
  properties {
   attenuationSettings=/Game/Sounds/Asteria/Environment_Actors/LightCastle_FirePlace_A/Fort_Fireplace_Athena_Close_Att
   attenuationOverrides {
    distanceAlgorithm=Custom
    falloffDistance=0
    attenuationShapeExtents {x=0}
    customAttenuationCurve {editorCurveData {keys [{interpMode=RCIM_Cubic, tangentMode=RCTM_Break, time=0, value=1}, {interpMode=RCIM_Cubic, tangentMode=RCTM_Break, time=1, value=0}]}}
   }
  }
 }
 {
  id=Attenuation_3
  children [WavePlayer_0]
  properties {
   attenuationSettings=&Fort_Fireplace_Athena_Close_Att
   attenuationOverrides {
    distanceAlgorithm=Custom
    falloffDistance=0
    attenuationShapeExtents {x=0}
    customAttenuationCurve {editorCurveData {keys [{interpMode=RCIM_Cubic, tangentMode=RCTM_Break, time=0, value=1}, {interpMode=RCIM_Cubic, tangentMode=RCTM_Break, time=1, value=0}]}}
   }
  }
 }
 {id=Mixer_0, children [Attenuation_3, Attenuation_2, Attenuation_1], properties {inputVolume [1,1,0.75]}}
 {id=Mixer_2, children [Mixer_0, Attenuation_1], properties {inputVolume [1,1]}}
 {id=QualityLevel_1, children [Mixer_2,Mixer_2,Mixer_0,Mixer_0]}
 {id=WavePlayer_0, properties {bLooping=true, soundWaveAssetPtr=/Game/Athena/Sounds/Events/Winterfest/Winterfest_Fireplace_Close_Loop}}
 {
  id=WavePlayer_1
  properties {bLooping=true, soundWaveAssetPtr=/Game/Sounds/Artemis/Environment_Actors/ColiseumBrazierFire/Spot_Coliseum_Brazier_Fire_Close_Mono}
 }
 {id=WavePlayer_6, properties {bLooping=true, soundWaveAssetPtr=/Game/Sounds/Fort_GamePlay_Sounds/Curie_Fires/Curie_Fire_MediumLoop_03}}
]
properties {
 bOverrideConcurrency=true
 maxDistance=2300
 soundClassObject=/Game/Sounds/FortSoundClassesAndModes/Fort_Ambient_NonLoop
 volumeMultiplier=0.25
 concurrencyOverrides {maxCount=2, volumeScaleMode=Distance}
}
